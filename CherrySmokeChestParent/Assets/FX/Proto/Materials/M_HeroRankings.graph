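# M_HeroRankings is an additive, unlit Niagara sprite material that draws a frame from the hero rankings texture.
# A SubUV function treats the texture as a four-by-one strip, with the Frame parameter choosing the cell.
# A dot product with the RGBChannel vector isolates one channel, red by default, as a greyscale mask.
# That mask multiplies particle color and particle alpha separately, so the particle system controls tint and fade.
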Material /CherrySmokeChestParent/Assets/FX/Proto/Materials/M_HeroRankings
graph {
 properties {
  bAutomaticallySetUsageInEditor=false
  bUsedWithNiagaraSprites=true
  blendMode=BLEND_Additive
  shadingModel=MSM_Unlit
  usedShadingModels=Unlit
 }
 nodes [
  n1 = MaterialFunctionCall!(
   Texture=TextureObject() {texture=/CherrySmokeChestParent/Assets/FX/Proto/T_HeroRankings}
   UVs
   SubImages=Constant2Vector() {g=1, r=4}
   Frame=ScalarParameter() {parameterName=Frame}
  ) {materialFunction=/Engine/Functions/Engine_MaterialFunctions03/Texturing/SubUV_Function}
  MaterialFunctionCall(Float3=@n1.RGB) {materialFunction=&BreakOutFloat3Components}
  n2 = ParticleColor()
  n3 = DotProduct(A=@n1.RGB, B=VectorParameter() {parameterName=RGBChannel, defaultValue=<rgba 1 0 0 1>}.RGB)
  Multiply(A=@n2.RGB, B=@n3)
  Multiply(A=@n2.A, B=@n3)
 ]
}
