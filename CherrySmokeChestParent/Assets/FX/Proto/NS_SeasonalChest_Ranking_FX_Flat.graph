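# Niagara system for a seasonal chest ranking effect, apparently a flat holographic variant in the prototype FX folder.
# Active emitters are the ranking, its backdrop and two HoloBits layers; the HoloBeams variants, a third HoloBits and Squares_LW are disabled.
# It exposes User.Rank and User.Unlocked parameters.
# The system uses a Timeline module, a system-spawn variable assignment, and color and Vector2 lerp inputs, with their wiring not shown.
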NiagaraSystem /CherrySmokeChestParent/Assets/FX/Proto/NS_SeasonalChest_Ranking_FX_Flat
dependencies {
 usedDataInterfaces []
 usedDynamicInputs [&LerpLinearColors, &Lerp_Vector2]
 usedModules [
  &SystemState
  &Timeline
  &self:SystemSpawnScript.SystemScriptSource.SystemScriptGraph.NiagaraNodeAssignment_0.SetVariables_#1
 ]
}
emitters {
 Em_HoloBeams {bEnabled=false, rendererClasses [], renderers []}
 Em_HoloBeams_Wide {bEnabled=false, rendererClasses [], renderers []}
 Em_HoloBeams_Wide001 {bEnabled=false, rendererClasses [], renderers []}
 Em_HoloBeams_Wide002 {bEnabled=false, rendererClasses [], renderers []}
 Em_HoloBeams_Wide003 {bEnabled=false, rendererClasses [], renderers []}
 Em_HoloBits {rendererClasses [], renderers []}
 Em_HoloBits001 {rendererClasses [], renderers []}
 Em_HoloBits002 {bEnabled=false, rendererClasses [], renderers []}
 Em_Ranking {rendererClasses [], renderers []}
 Em_Ranking_Backdrop {rendererClasses [], renderers []}
 Squares_LW {bEnabled=false, rendererClasses [], renderers []}
}
summary {
 systemName=NS_SeasonalChest_Ranking_FX_Flat
 emitters [
  {emitterName=Em_Ranking, rendererClasses []}
  {bEnabled=false, emitterName=Em_HoloBeams, rendererClasses []}
  {bEnabled=false, emitterName=Em_HoloBeams_Wide, rendererClasses []}
  {bEnabled=false, emitterName=Squares_LW, rendererClasses []}
  {emitterName=Em_HoloBits, rendererClasses []}
  {bEnabled=false, emitterName=Em_HoloBeams_Wide001, rendererClasses []}
  {emitterName=Em_HoloBits001, rendererClasses []}
  {bEnabled=false, emitterName=Em_HoloBits002, rendererClasses []}
  {bEnabled=false, emitterName=Em_HoloBeams_Wide002, rendererClasses []}
  {bEnabled=false, emitterName=Em_HoloBeams_Wide003, rendererClasses []}
  {emitterName=Em_Ranking_Backdrop, rendererClasses []}
 ]
 userVariables [
  User.Rank {type=NiagaraFloat, defaultValue=0, description=""}
  User.Unlocked {type=NiagaraBool, defaultValue=true}
 ]
}
